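# Test material for sprite-sheet imposters, masked Default Lit with non-tangent-space normals and instanced static mesh support.
# ImposterUVs takes frame grid, size, rotation, object position and a normal map, and its UVs drive the BaseColor sample.
# Motion-vector flipbook, foliage scale and rotation, frame-offset and debug nodes sit largely unconnected as apparent experiments.
# The graph shows no connections to material outputs.
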Material /Engine/ArtTools/RenderToTexture/Materials/Imposters/M_ImposterTest_01
graph {
 properties {
  bTangentSpaceNormal=false
  bUsedWithInstancedStaticMeshes=true
  blendMode=BLEND_Masked
  numCustomizedUVs=4
  pixelDepthOffsetMode=PDOM_Legacy
  usedShadingModels=Default Lit
 }
 nodes [
  n1 = ScalarParameter() {defaultValue=20, parameterName=Columns}
  n2 = ScalarParameter() {defaultValue=20, parameterName=Rows}
  n3 = ScalarParameter() {defaultValue=141, parameterName=Size}
  n4 = ScalarParameter() {parameterName=Rot}
  n5 = MaterialFunctionCall!(
   FramesX=@n1
   FramesY=@n2
   Fixed Z=StaticBoolParameter() {parameterName=FixedZ}
   SpriteWidth=@n3
   Position=ObjectPositionWS()
   SpriteHeight=@n3
   Rotation=@n4
   Rotation For Normals=@n4
   Normals=TextureSampleParameter2D() {parameterName=Normal, samplerType=SAMPLERTYPE_Normal, texture=/Engine/EngineMaterials/FlatNormal}.RGB
  ) {materialFunction=/Engine/ArtTools/RenderToTexture/MaterialFunctions/ImposterUVs}
  n6 = Constant()
  Add(A=@n5.UVs, B=AppendVector(A=Divide(B=@n1), B=@n6)) {desc=Imposter X offset}
  Add(A=@n5.UVs, B=AppendVector(A=@n6, B=Divide(B=@n2))) {desc=Imposter Y offset}
  AppendVector(A=@n1, B=@n2)
  ComponentMask(None=TextureCoordinate() {coordinateIndex=2}) {r=true}
  Constant() {r=1}
  Constant() {r=100}
  n7 = LinearInterpolate()
  LinearInterpolate(Alpha=@n7)
  LinearInterpolate(Alpha=@n7)
  MaterialFunctionCall(Number=ComponentMask(None=@n5.BlendPhase xy) {r=true}) {materialFunction=&DebugScalarValues}
  MaterialFunctionCall!(
   FramesX
   FramesY
   Fixed Z
   SpriteWidth
   Position
   SpriteHeight
   Rotation
   Rotation For Normals
   Motion Vector Intensity
   Normals
   Base Texture
   Motion Vector Texture
  ) {materialFunction=/Engine/ArtTools/RenderToTexture/MaterialFunctions/Imposter_MotionVectors}
  MaterialFunctionCall!(
   "Animation  Phase (0-1)"
   Number of Columns X
   Number of Rows Y
   Texture
   UVs
   Motion Vector Texture
   Phase to Debug
   Motion Vector Intensity
   Double Apply Motion Vectors
   Custom UV Input
   Custom UV2 Input
   Use Custom UVs
  ) {materialFunction=/Engine/Functions/Engine_MaterialFunctions02/Texturing/FlipBook_MotionVectors}
  MaterialFunctionCall!(Initial Radius=@n3) {materialFunction=/Engine/Functions/Engine_MaterialFunctions01/WorldPositionOffset/FoliageScaleFactor}
  MaterialFunctionCall() {materialFunction=/Engine/ArtTools/RenderToTexture/MaterialFunctions/FoliageZRotation}
  ScalarParameter() {parameterName=R}
  ScalarParameter() {parameterName=R}
  TextureSampleParameter2D(UVs=@n5.UVs) {parameterName=BaseColor, samplerType=SAMPLERTYPE_LinearColor, texture=/Engine/ArtTools/RenderToTexture/Textures/TilingTexturePreview}
  TransformPosition(None=Add(A=Constant3Vector(), B=VectorParameter() {parameterName=Offset}.RGB))
 ]
}
